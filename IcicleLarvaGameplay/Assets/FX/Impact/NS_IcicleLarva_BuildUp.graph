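# Niagara build-up effect for the Icicle Larva: brief, delayed sprite bursts that shrink while fading in, then cut out.
# FakeVolume is a large camera-offset sprite whose alpha envelope is modulated by a flicker curve, sampled on repeating normalized age.
# ShockWave, delayed by System.BurstTime, has three sprite renderers, two extra curve-scaled size attributes and a curve-driven dynamic material parameter.
# A Sonic_Boom_LW emitter is listed without renderers.
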NiagaraSystem /IcicleLarvaGameplay/Assets/FX/Impact/NS_IcicleLarva_BuildUp
dependencies {usedDataInterfaces [&NiagaraDataInterfaceCurve], usedDynamicInputs [], usedModules [&SystemState]}
emitters {
 FakeVolume {
  emitterUpdateScript [
   &EmitterState(Life Cycle Mode=::NewEnumerator1, Loop Behavior=::NewEnumerator1, Loop Delay=0.2, Loop Duration=0.15, UseLoopDelay=true) {module=EmitterState001}
   &SpawnBurst_Instantaneous(Use Spawn Probability=false, UseLoopCountLimit=false)
  ]
  particleSpawnScript [&InitializeParticle(
    Color Mode=::NewEnumerator0
    Lifetime=0.35
    Sprite Size Mode=::NewEnumerator3
    Uniform Sprite Size=500
    UsePositionOffset=false
    Write Material Random=true
   )]
  particleUpdateScript [
   &ParticleState()
   &ScaleSpriteSize(Uniform Curve Sprite Scale=DataInterface() {propertyValues {Curve {
       defaultValue=0
       keys [<at 0 1 RCIM_Cubic>, {arriveTangent=-0.0138, interpMode=RCIM_Cubic, leaveTangent=-0.0138, tangentMode=RCTM_User, time=1, value=0.35}]
      }}})
   &DynamicMaterialParameters(
    Index 0 Param 3=1
    Index 0 Param 4=500
    Module.Index0Float0Write=true
    Module.Index0Float1Write=true
    Module.Index0Float2Write=true
    Module.Index0Float3Write=true
   )
   &ScaleColor(
    Scale Alpha=&FloatFromCurve(
     FloatCurve=DataInterface() {propertyValues {Curve {
        defaultValue=0
        keys [
         {interpMode=RCIM_Cubic, tangentMode=RCTM_User, time=0, value=0}
         {interpMode=RCIM_Cubic, tangentMode=RCTM_User, time=0.8, value=1}
         {tangentMode=RCTM_User, time=1, value=0}
        ]
       }}}
     Scale Curve=&FloatFromCurve(
      FloatCurve=DataInterface() {propertyValues {Curve {keys [
          <at 0 1 RCIM_Cubic>
          <at 0.1 0.4 RCIM_Cubic>
          <at 0.2 1 RCIM_Cubic>
          <at 0.3 0.4 RCIM_Cubic>
          <at 0.4 1 RCIM_Cubic>
          <at 0.5 0.4 RCIM_Cubic>
          <at 0.6 1 RCIM_Cubic>
          <at 0.7 0.4 RCIM_Cubic>
          {arriveTangent=2, interpMode=RCIM_Cubic, leaveTangent=2, time=0.9, value=1}
          <at 1 1 RCIM_Cubic>
         ]}}}
      CurveIndex=&Frac_Float(Float=&Multiply_Float(A=$Particles.NormalizedAge, B=4))
      Scale Curve=0.2
     )
    )
    ScaleA=true
    ScaleRGB=true
   )
   &JitterPosition(Jitter Amount=300) {enabled=false}
   &CameraOffset(Camera Offset Amount=-100)
  ]
  rendererClasses [&NiagaraSpriteRendererProperties]
  renderers [{rendererClass=&NiagaraSpriteRendererProperties}]
 }
 ShockWave {
  emitterUpdateScript [
   &EmitterState(Life Cycle Mode=::NewEnumerator1, Loop Behavior=::NewEnumerator1, Loop Delay=$System.BurstTime, Loop Duration=1, UseLoopDelay=true)
   &SpawnBurst_Instantaneous(Use Spawn Probability=false, UseLoopCountLimit=false)
  ]
  particleSpawnScript [&InitializeParticle(
    Color=<rgba 0.1163 0.1163 0.1163 0.1>
    Lifetime=0.5
    Sprite Size Mode=::NewEnumerator3
    Uniform Sprite Size=300
    UsePositionOffset=false
    Write Material Random=true
   )]
  particleUpdateScript [
   &ParticleState()
   &ScaleSpriteSize(Uniform Curve Sprite Scale=DataInterface() {propertyValues {Curve {
       defaultValue=0.25
       keys [{arriveTangent=-1.3311, interpMode=RCIM_Cubic, leaveTangent=-1.3311, tangentMode=RCTM_User, time=0, value=1}, <at 1 0.6 RCIM_Cubic>]
      }}})
   &DynamicMaterialParameters(
    Index 0 Param 4=&FloatFromCurve(
     FloatCurve=DataInterface() {propertyValues {Curve {
        defaultValue=1
        keys [
         {arriveTangent=0.0405, interpMode=RCIM_Cubic, leaveTangent=0.0405, leaveTangentWeight=0.3002, tangentMode=RCTM_User, time=0, value=0}
         {
          arriveTangent=1.4853
          arriveTangentWeight=0.5372
          interpMode=RCIM_Cubic
          leaveTangent=1.4853
          leaveTangentWeight=0.0597
          tangentMode=RCTM_User
          time=0.9
          value=0.975
         }
         {arriveTangent=-9.0725, arriveTangentWeight=0.3042, interpMode=RCIM_Cubic, leaveTangent=-9.0725, tangentMode=RCTM_User, time=1, value=0}
        ]
       }}}
     Scale Curve=0.7
    )
    Module.Index0Float0Write=true
    Module.Index0Float1Write=true
    Module.Index0Float2Write=true
    Module.Index0Float3Write=true
   )
   /SetVariables(
    Particles.SpriteSize001:Vector2f=&Multiply_Vector2D_ByFloat(
     Float=&FloatFromCurve(FloatCurve=DataInterface() {propertyValues {Curve {defaultValue=1, keys [{tangentMode=RCTM_User, time=0, value=3}, {tangentMode=RCTM_User, time=1, value=2}]}}})
     Vector2D=$Particles.SpriteSize
    )
    Particles.SpriteSize002:Vector2f=&Multiply_Vector2D_ByFloat(
     Float=&FloatFromCurve(FloatCurve=DataInterface() {propertyValues {Curve {defaultValue=1, keys [{tangentMode=RCTM_User, time=0, value=0.75}, {tangentMode=RCTM_User, time=1, value=0.25}]}}})
     Vector2D=$Particles.SpriteSize
    )
   ) {isSetParameters=true}
   &ScaleColor(
    Scale Alpha=&FloatFromCurve(FloatCurve=DataInterface() {propertyValues {Curve {
        defaultValue=0
        keys [
         {arriveTangent=0.0405, interpMode=RCIM_Cubic, leaveTangent=0.0405, leaveTangentWeight=0.2866, tangentMode=RCTM_User, time=0, value=0}
         {
          arriveTangent=2.5249
          arriveTangentWeight=0.7778
          interpMode=RCIM_Cubic
          leaveTangent=2.5249
          leaveTangentWeight=0.1274
          tangentMode=RCTM_User
          time=0.8592
          value=0.9847
         }
         {arriveTangent=-9.0725, arriveTangentWeight=0.4283, interpMode=RCIM_Cubic, leaveTangent=-9.0725, tangentMode=RCTM_User, time=1, value=0}
        ]
       }}})
    ScaleA=true
    ScaleRGB=true
   )
  ]
  rendererClasses [&NiagaraSpriteRendererProperties]
  renderers [
   {rendererClass=&NiagaraSpriteRendererProperties}
   {rendererClass=&NiagaraSpriteRendererProperties, rendererIndex=1}
   {rendererClass=&NiagaraSpriteRendererProperties, rendererIndex=2}
  ]
 }
 Sonic_Boom_LW {rendererClasses [], renderers []}
}
summary {
 systemName=NS_IcicleLarva_BuildUp
 emitters [
  {emitterName=ShockWave, rendererClasses [&NiagaraSpriteRendererProperties]}
  {emitterName=FakeVolume, rendererClasses [&NiagaraSpriteRendererProperties]}
  {emitterName=Sonic_Boom_LW, rendererClasses []}
 ]
 userVariables []
}
